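# Eye material for the male medium HornedJudgment Midgard character body, built on the shared MF_eyes function with a custom emissive color.
# It runs the function's emissive output through a Step to form a mask, then multiplies it by AdvancedEmissiveColor and EmissiveIntensity.
# The result replaces the emissive attribute before MF_TechArtOverrides finishes the material.
# The color defaults to black, so the eyes emit nothing unless that parameter is overridden.
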Material /BRCosmetics/Characters/Player/Male/Medium/Bodies/M_MED_HornedJudgment_Midgard/Materials/M_Eyes_HornedJudgment
graph {
 properties {
  bUseMaterialAttributes=true
  bUsedWithMeshDeformer=true
  bUsedWithSkeletalMesh=true
  pixelDepthOffsetMode=PDOM_Legacy
  shadingModel=MSM_FromMaterialExpression
  usedShadingModels=Default Lit
 }
 nodes [
  n1 = GetMaterialAttributes(None=MaterialFunctionCall() {materialFunction=/Game/Characters/Shared/FNEye/Materials/MF_eyes}.Result) {attributeGetTypes [#1]}
  n2 = Step(Y=@n1.Emissive Color)
  n3 = VectorParameter() {group=HornedJudgment Emissive, parameterName=AdvancedEmissiveColor, sortPriority=1, defaultValue=<rgba 0 0 0 1>}
  n4 = Multiply(A=@n3.RGB, B=ScalarParameter() {defaultValue=2, group=HornedJudgment Emissive, parameterName=EmissiveIntensity, sortPriority=1}) {constB=2}
  MaterialFunctionCall(In=SetMaterialAttributes(MaterialAttributes=@n1.MaterialAttributes, Emissive Color=Multiply(A=@n2, B=@n4)) {attributeSetTypes [#1]}) {materialFunction=&MF_TechArtOverrides}
 ]
}
